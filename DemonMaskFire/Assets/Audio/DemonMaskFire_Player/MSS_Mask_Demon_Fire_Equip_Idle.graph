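# Stereo MetaSound source that plays the equip and idle audio for the Demon Fire mask on a Fortnite character.
# On play, it fires a random appear sound and starts the idle loop.
# Each footstep trigger runs an ADSR envelope that modulates the loop's pitch and gain.
# A mixer combines both layers with affiliation-based gains that boost enemies; the jump, health and ID inputs are unused.
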MetaSoundSource /DemonMaskFire/Assets/Audio/DemonMaskFire_Player/MSS_Mask_Demon_Fire_Equip_Idle
inputs [
 Character.Affiliation {type=Enum:FortAudioAffiliation}
 Character.Health {type=Float}
 Character.UniqueID {type=Int32}
 Trigger.OnJump {type=Float}
 Trigger.OnStep {type=Trigger}
 UE.Source.OnPlay {type=Trigger}
 Waves.IdleLoop {type=WaveAsset:Array}
 Waves.OnAppear {type=WaveAsset:Array}
]
interfaces [Fort.Character:1.0, UE.OutputFormat.Stereo:1.0, UE.Source:1.1]
outputs [
 "UE.OutputFormat.Stereo.Audio:0" {type=Audio}
 "UE.OutputFormat.Stereo.Audio:1" {type=Audio}
]
pages [Default {
  input_defaults {
   Character.Affiliation=0
   Character.Health=0
   Character.UniqueID=0
   Trigger.OnJump=0
   Trigger.OnStep=null
   UE.Source.OnPlay=null
   Waves.IdleLoop [&/Mask_Demon_Fire_Idle_Loop_01]
   Waves.OnAppear [&/Mask_Demon_Fire_Equip_01]
  }
  nodes [
   ADSREnvelopeADSREnvelope_0 = ADSR Envelope.ADSR Envelope:Float(Trigger Attack=^Trigger.OnStep) {also {defaults {Attack Time=0.25, Decay Curve=0.5, Decay Time=2, Release Time=2, Sustain Level=0}}}
   n1 = &MSP_Audio_WavePlayer_Deluxe_Stereo(
    Play Trigger=^UE.Source.OnPlay
    Wave Asset=^Waves.IdleLoop
    "Pitch Shift Global (Semitones)"=/FortMetasound/SoundCue/MSP_ContinuousModulator(Value In=@ADSREnvelopeADSREnvelope_0.Out Envelope) {also {defaults {Output.Max=2, Output.Min=-2}}}.Value Out
    WavePlayer Gain dB=&MSP_ContinuousModulator(Value In=@ADSREnvelopeADSREnvelope_0.Out Envelope) {also {defaults {Output.Max=0, Output.Min=-9}}}.Value Out
   ) {also {defaults {Is Looping?="<MetaSoundBuilderResult.SUCCEEDED: 0>"}}}
   n2 = &MSS_PlayRandomSound_Stereo(UE.Source.OnPlay=^UE.Source.OnPlay, Sounds=^Waves.OnAppear)
   AudioMixerAudioMixerStereo3_0 = "AudioMixer.Audio Mixer (Stereo, 3)"(
    Gain 0=&MSP_AffiliationMixFloat(Affiliation=^Character.Affiliation) {also {defaults {Enemy=1.5}}}.Out
    Gain 1=&MSP_AffiliationMixFloat(Affiliation=^Character.Affiliation) {also {defaults {Enemy=1.5}}}.Out
    In 1 L=@n1.Out L
    In 1 R=@n1.Out R
    In 0 L=@n2."UE.OutputFormat.Stereo.Audio:0"
    In 0 R=@n2."UE.OutputFormat.Stereo.Audio:1"
   ) {also {defaults {Gain 2=0.3}}}
  ]
  links [
   {dst="@UE.OutputFormat.Stereo.Audio:0", src=AudioMixerAudioMixerStereo3_0.Out L}
   {dst="@UE.OutputFormat.Stereo.Audio:1", src=AudioMixerAudioMixerStereo3_0.Out R}
  ]
 }]
properties {
 assetImportData=&self:AssetImportData
 attenuationSettings=&Fort_Weapon_Reload_Att
 bLooping=true
 duration=10000
 maxDistance=2250
 numChannels=2
 outputFormat=Stereo
 sampleRate=48000
 soundClassObject=&Fort_WeaponFoley
 volume=0.7
 concurrencyOverrides {bLimitToOwner=true, maxCount=1, resolutionRule=StopOldest}
}
