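# Sound cue for the drop sound of the player's Demon Fire Mask.
# A single wave player plays one drop sample through a modulator at its default settings, at reduced volume.
# It uses weapon reload attenuation and the weapon foley sound class, with a concurrency override that limits playback to one instance at a time.
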SoundCue /DemonMaskFire/Assets/Audio/DemonMaskFire_Player/SC_Mask_Demon_Fire_Player_Drop
root=Modulator_0
nodes [
 {id=Modulator_0, children [WavePlayer_0]}
 {id=WavePlayer_0, properties {soundWaveAssetPtr=&/Mask_Demon_Fire_Drop_01}}
]
properties {
 attenuationSettings=&Fort_Weapon_Reload_Att
 bOverrideConcurrency=true
 maxDistance=2250
 soundClassObject=&Fort_WeaponFoley
 volumeMultiplier=0.35
 concurrencyOverrides {maxCount=1}
}
