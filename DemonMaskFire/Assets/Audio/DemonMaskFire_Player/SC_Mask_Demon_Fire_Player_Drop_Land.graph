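# Sound cue for the player's Demon Fire Mask drop-land sound, judging by its name.
# A single wave player feeds the Mask_Demon_Fire_Drop_Land_01 sample through a modulator left at default settings.
# The cue plays at reduced volume in the weapon foley sound class and borrows weapon reload attenuation for distance falloff.
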SoundCue /DemonMaskFire/Assets/Audio/DemonMaskFire_Player/SC_Mask_Demon_Fire_Player_Drop_Land
root=Modulator_0
nodes [
 {id=Modulator_0, children [WavePlayer_0]}
 {id=WavePlayer_0, properties {soundWaveAssetPtr=&/Mask_Demon_Fire_Drop_Land_01}}
]
properties {
 attenuationSettings=&Fort_Weapon_Reload_Att
 maxDistance=2250
 soundClassObject=&Fort_WeaponFoley
 volumeMultiplier=0.5
}
