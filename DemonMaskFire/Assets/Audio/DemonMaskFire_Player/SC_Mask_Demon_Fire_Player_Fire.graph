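# Sound cue for the player's Demon Mask fire sound, played as a ranged weapon sound with its own attenuation settings.
# It randomly picks one of four equally weighted close-range fire waves.
# Nested branches on the islocal and isenemy parameters choose fixed-volume modulators, halving volume for non-local, non-enemy players.
# Local players, enemies and unset parameters get full volume, with no pitch variation on any path.
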SoundCue /DemonMaskFire/Assets/Audio/DemonMaskFire_Player/SC_Mask_Demon_Fire_Player_Fire
root=Branch_2
nodes [
 {id=Branch_2, children [Modulator_2,Branch_3,Modulator_2], properties {boolParameterName=islocal}}
 {id=Branch_3, children [Modulator_4,Modulator_3,Modulator_4], properties {boolParameterName=isenemy}}
 {id=Modulator_2, children [Random_1], properties {pitchMax=1, pitchMin=1, volumeMax=1, volumeMin=1}}
 {id=Modulator_3, children [Random_1], properties {pitchMax=1, pitchMin=1, volumeMax=0.5, volumeMin=0.5}}
 {id=Modulator_4, children [Random_1], properties {pitchMax=1, pitchMin=1, volumeMax=1, volumeMin=1}}
 {id=Random_1, children [WavePlayer_0,WavePlayer_1,WavePlayer_2,WavePlayer_3], properties {weights [1,1,1,1]}}
 {id=WavePlayer_0, properties {soundWaveAssetPtr=&/Mask_Demon_Fire_Fire_Close_04}}
 {id=WavePlayer_1, properties {soundWaveAssetPtr=&/Mask_Demon_Fire_Fire_Close_03}}
 {id=WavePlayer_2, properties {soundWaveAssetPtr=&/Mask_Demon_Fire_Fire_Close_02}}
 {id=WavePlayer_3, properties {soundWaveAssetPtr=&/Mask_Demon_Fire_Fire_Close_01}}
]
properties {
 attenuationSettings=&/SA_Mask_Demon_Fire_Player
 maxDistance=10400
 soundClassObject=&Fort_WeaponRanged
 volumeMultiplier=1
}
