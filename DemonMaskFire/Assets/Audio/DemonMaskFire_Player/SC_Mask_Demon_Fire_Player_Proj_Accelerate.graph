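# Sound cue named for the accelerate stage of the player's Demon Mask fire projectile, classed as ranged weapon audio.
# The node graph is empty with no root, so the cue holds only playback settings and no sound of its own.
# It references a dedicated attenuation asset, also carries long-range falloff overrides with occlusion filtering, caps simultaneous instances, and sends lightly to a distant reverb submix.
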SoundCue /DemonMaskFire/Assets/Audio/DemonMaskFire_Player/SC_Mask_Demon_Fire_Player_Proj_Accelerate
root=null
nodes []
properties {
 attenuationSettings=&/SA_Mask_Demon_Fire_Player_Proj_Accelerate
 bOverrideConcurrency=true
 maxDistance=8750
 soundClassObject=&Fort_WeaponRanged
 volumeMultiplier=0.8
 attenuationOverrides {
  bEnableOcclusion=true
  falloffDistance=8000
  occlusionLowPassFilterFrequency=1500
  occlusionTraceChannel=ECC_GameTraceChannel6
  attenuationShapeExtents {x=0}
  customAttenuationCurve {editorCurveData {keys [<at 0 1>,<at 0.05 1>,<at 0.2 0.5>,<at 1 0>]}}
 }
 concurrencyOverrides {maxCount=2}
 soundSubmixSends [{minSendLevel=0.2, sendLevel=0.2, soundSubmix=&Fortnite_Reverb_Distant}]
}
