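# Sound cue for the player's Demon Mask fire projectile passing close by, routed to the ranged weapon sound class.
# A mixer layers one of four random close-by fire whooshes with one of seven reused Puppeteer fireball screams at lower level, the scream varied in pitch and volume by a modulator.
# Overridden attenuation adds a custom falloff curve and low-pass occlusion, with tight concurrency and a distant reverb send.
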SoundCue /DemonMaskFire/Assets/Audio/DemonMaskFire_Player/SC_Mask_Demon_Fire_Player_Proj_By
root=Mixer_1
nodes [
 {id=Mixer_1, children [Random_1,Modulator_0], properties {inputVolume [1,0.6]}}
 {id=Modulator_0, children [Random_3], properties {pitchMax=0.8, pitchMin=1, volumeMin=0.8}}
 {id=Random_1, children [WavePlayer_0,WavePlayer_1,WavePlayer_2,WavePlayer_3], properties {weights [1,1,1,1]}}
 {
  id=Random_3
  children [WavePlayer_8, WavePlayer_9, WavePlayer_10, WavePlayer_11, WavePlayer_12, WavePlayer_13, WavePlayer_14]
  properties {weights [1,1,1,1,1,1,1]}
 }
 {id=WavePlayer_0, properties {soundWaveAssetPtr=&/Mask_Demon_Fire_Proj_By_Close_01}}
 {id=WavePlayer_1, properties {soundWaveAssetPtr=&/Mask_Demon_Fire_Proj_By_Close_02}}
 {
  id=WavePlayer_10
  properties {soundWaveAssetPtr=/DemonMaskFire/Assets/Audio/DemonMaskFire_Puppeteer/wav/Puppeteer_Mask_Fireball_By_Close_Scream_03}
 }
 {
  id=WavePlayer_11
  properties {soundWaveAssetPtr=/DemonMaskFire/Assets/Audio/DemonMaskFire_Puppeteer/wav/Puppeteer_Mask_Fireball_By_Close_Scream_04}
 }
 {
  id=WavePlayer_12
  properties {soundWaveAssetPtr=/DemonMaskFire/Assets/Audio/DemonMaskFire_Puppeteer/wav/Puppeteer_Mask_Fireball_By_Close_Scream_05}
 }
 {
  id=WavePlayer_13
  properties {soundWaveAssetPtr=/DemonMaskFire/Assets/Audio/DemonMaskFire_Puppeteer/wav/Puppeteer_Mask_Fireball_By_Close_Scream_06}
 }
 {
  id=WavePlayer_14
  properties {soundWaveAssetPtr=/DemonMaskFire/Assets/Audio/DemonMaskFire_Puppeteer/wav/Puppeteer_Mask_Fireball_By_Close_Scream_07}
 }
 {id=WavePlayer_2, properties {soundWaveAssetPtr=&/Mask_Demon_Fire_Proj_By_Close_03}}
 {id=WavePlayer_3, properties {soundWaveAssetPtr=&/Mask_Demon_Fire_Proj_By_Close_04}}
 {
  id=WavePlayer_8
  properties {soundWaveAssetPtr=/DemonMaskFire/Assets/Audio/DemonMaskFire_Puppeteer/wav/Puppeteer_Mask_Fireball_By_Close_Scream_01}
 }
 {
  id=WavePlayer_9
  properties {soundWaveAssetPtr=/DemonMaskFire/Assets/Audio/DemonMaskFire_Puppeteer/wav/Puppeteer_Mask_Fireball_By_Close_Scream_02}
 }
]
properties {
 attenuationSettings=&/SA_Mask_Demon_Fire_Player
 bOverrideAttenuation=true
 bOverrideConcurrency=true
 maxDistance=8000
 soundClassObject=&Fort_WeaponRanged
 volumeMultiplier=0.8
 attenuationOverrides {
  bEnableOcclusion=true
  falloffDistance=8000
  occlusionLowPassFilterFrequency=1500
  occlusionTraceChannel=ECC_GameTraceChannel6
  attenuationShapeExtents {x=0}
  customAttenuationCurve {editorCurveData {keys [<at 0 1>,<at 0.05 1>,<at 0.2 0.5>,<at 1 0>]}}
 }
 concurrencyOverrides {maxCount=2}
 soundSubmixSends [{minSendLevel=0.2, sendLevel=0.2, soundSubmix=&Fortnite_Reverb_Distant}]
}
