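# Sound cue for the Demon Mask fire projectile passing close by, in the player's first-person (1P) variant on the ranged weapon sound class.
# A random pick of four close fly-by waves feeds a short fixed delay, then volume modulators selected by islocal and isenemy branches.
# Non-local playback is quieter, with enemy and non-enemy paths currently identical.
# The cue lowers pitch, uses projectile attenuation and caps concurrency, stopping the oldest.
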SoundCue /DemonMaskFire/Assets/Audio/DemonMaskFire_Player/SC_Mask_Demon_Fire_Player_Proj_By_1P
root=Branch_2
nodes [
 {id=Branch_2, children [Modulator_2,Branch_3,Modulator_2], properties {boolParameterName=islocal}}
 {id=Branch_3, children [Modulator_4,Modulator_3,Modulator_4], properties {boolParameterName=isenemy}}
 {id=Delay_0, children [Random_1], properties {delayMax=0.1, delayMin=0.1}}
 {id=Modulator_2, children [Delay_0], properties {pitchMax=1, pitchMin=1, volumeMax=1, volumeMin=1}}
 {id=Modulator_3, children [Delay_0], properties {pitchMax=1, pitchMin=1, volumeMax=0.4, volumeMin=0.4}}
 {id=Modulator_4, children [Delay_0], properties {pitchMax=1, pitchMin=1, volumeMax=0.4, volumeMin=0.4}}
 {id=Random_1, children [WavePlayer_0,WavePlayer_1,WavePlayer_2,WavePlayer_3], properties {weights [1,1,1,1]}}
 {id=WavePlayer_0, properties {soundWaveAssetPtr=&/Mask_Demon_Fire_Proj_By_Close_01}}
 {id=WavePlayer_1, properties {soundWaveAssetPtr=&/Mask_Demon_Fire_Proj_By_Close_02}}
 {id=WavePlayer_2, properties {soundWaveAssetPtr=&/Mask_Demon_Fire_Proj_By_Close_03}}
 {id=WavePlayer_3, properties {soundWaveAssetPtr=&/Mask_Demon_Fire_Proj_By_Close_04}}
]
properties {
 attenuationSettings=&/SA_Mask_Demon_Fire_Player_Proj
 bOverrideConcurrency=true
 maxDistance=6500
 pitchMultiplier=0.7
 soundClassObject=&Fort_WeaponRanged
 volumeMultiplier=0.5
 concurrencyOverrides {maxCount=2, resolutionRule=StopOldest, retriggerTime=0.2}
}
