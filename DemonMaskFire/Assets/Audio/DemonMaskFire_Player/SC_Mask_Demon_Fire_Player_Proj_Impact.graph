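# Sound cue for the impact of the player's Demon Mask fire projectile.
# A Random node picks one of four close fire-impact waves with equal weight, so repeated hits vary.
# It plays in the ranged weapon sound class and uses a dedicated projectile-impact attenuation asset for spatial falloff.
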SoundCue /DemonMaskFire/Assets/Audio/DemonMaskFire_Player/SC_Mask_Demon_Fire_Player_Proj_Impact
root=Random_0
nodes [
 {id=Random_0, children [WavePlayer_4,WavePlayer_5,WavePlayer_6,WavePlayer_7], properties {weights [1,1,1,1]}}
 {id=WavePlayer_4, properties {soundWaveAssetPtr=&/Mask_Demon_Fire_Impact_Close_01}}
 {id=WavePlayer_5, properties {soundWaveAssetPtr=&/Mask_Demon_Fire_Impact_Close_02}}
 {id=WavePlayer_6, properties {soundWaveAssetPtr=&/Mask_Demon_Fire_Impact_Close_03}}
 {id=WavePlayer_7, properties {soundWaveAssetPtr=&/Mask_Demon_Fire_Impact_Close_04}}
]
properties {
 attenuationSettings=&/SA_Mask_Demon_Fire_Player_Proj_Impt
 maxDistance=12000
 soundClassObject=&Fort_WeaponRanged
}
