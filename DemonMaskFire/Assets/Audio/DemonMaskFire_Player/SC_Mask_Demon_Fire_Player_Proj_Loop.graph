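# Sound cue for the looping fire projectile sound of the player's Demon Mask, with attenuation and a tight concurrency limit.
# One looping wave runs through an enveloper that fades volume in, feeding three shared modulator variants.
# Branches on islocal, then isenemy, select the variant: local gets raised pitch, while others get lowered pitch, with enemies loudest and non-enemies quietest.
# Unset parameters fall back to the local and enemy variants respectively.
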SoundCue /DemonMaskFire/Assets/Audio/DemonMaskFire_Player/SC_Mask_Demon_Fire_Player_Proj_Loop
root=Branch_2
nodes [
 {id=Branch_2, children [Modulator_2,Branch_3,Modulator_2], properties {boolParameterName=islocal}}
 {id=Branch_3, children [Modulator_4,Modulator_3,Modulator_4], properties {boolParameterName=isenemy}}
 {
  id=Enveloper_0
  children [WavePlayer_0]
  properties {pitchCurve {editorCurveData {keys [{time=0}]}}, volumeCurve {editorCurveData {keys [<at 0 0>,<at 0.0959 0.2448>,<at 0.25 1>]}}}
 }
 {id=Modulator_2, children [Enveloper_0], properties {pitchMax=1.1, pitchMin=1, volumeMax=0.7, volumeMin=0.7}}
 {id=Modulator_3, children [Enveloper_0], properties {pitchMax=1, pitchMin=0.8, volumeMax=0.5, volumeMin=0.5}}
 {id=Modulator_4, children [Enveloper_0], properties {pitchMax=1, pitchMin=0.8, volumeMax=1, volumeMin=1}}
 {id=WavePlayer_0, properties {bLooping=true, soundWaveAssetPtr=&/Mask_Demon_Fire_Proj_Loop_01}}
]
properties {
 attenuationSettings=&/SA_Mask_Demon_Fire_Player_Proj
 bOverrideConcurrency=true
 maxDistance=6500
 soundClassObject=&Fort_WeaponRanged
 volumeMultiplier=0.7
 concurrencyOverrides {maxCount=2, retriggerTime=0.1}
}
