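# Sound cue for the recharge sound of the player's fire demon mask.
# A single wave player feeds a Modulator at default settings, which varies pitch and volume on each play.
# The cue plays at reduced volume in the weapon foley sound class with fire mask foley attenuation.
# A concurrency override limits it to one instance at a time.
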SoundCue /DemonMaskFire/Assets/Audio/DemonMaskFire_Player/SC_Mask_Demon_Fire_Player_Recharge
root=Modulator_0
nodes [
 {id=Modulator_0, children [WavePlayer_0]}
 {id=WavePlayer_0, properties {soundWaveAssetPtr=&/Mask_Demon_Recharge_01}}
]
properties {
 attenuationSettings=&/SA_Mask_Fire_Foley
 bOverrideConcurrency=true
 maxDistance=2000
 soundClassObject=&Fort_WeaponFoley
 volumeMultiplier=0.5
 concurrencyOverrides {maxCount=1}
}
